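# Animation Blueprint for the Aztec backpack cosmetic that poses its wings to match the player's movement.
# The WingPosition state machine selects pose-asset poses for base, targeting, crouching, jumping and falling, adding a directional lean blendspace in Base.
# A nested skydiving machine switches between parachute, dive and float poses.
# The cached result gets a speed-weighted additive Sprint pose, bypassed through bool blends while skydiving or crouching.
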AnimBlueprint /BRCosmetics/Accessories/FORT_Backpacks/F_MED_Aztec_Backpack/Meshes/F_MED_Aztec_Backpack_AnimBlueprint
target_skeleton=&/F_MED_Aztec_Backpack_Skeleton
graphs [
 AnimGraph {nodes [
   n1 = Wing_Movement_Poses_Add_PoseAsset : Sprint!(PoseWeight:float=Get SprintSpeedRemapped() -> (SprintSpeedRemapped:float)) -> (Pose:Pose Link Structure) {
    is=title AnimGraphNode_PoseByName
    poseAsset=/BRCosmetics/Accessories/FORT_Backpacks/F_MED_Aztec_Backpack/Animations/Wing_Movement_Poses_Add_PoseAsset
    poseName=Sprint
   }
   n2 = Apply Additive(Base=Use cached pose 'preSprint'(), Additive=@n1) {alphaBoolBlend {blendOption=Linear}}
   n3 = Use cached pose 'preSprint'()
   n4 = Get bIsCrouching()
   Output Pose(Result=Blend Poses by bool(
     BlendPose_0=Use cached pose 'preSprint'()
     BlendPose_1=Blend Poses by bool(BlendPose_0=@n3, BlendPose_1=@n2, BlendTime_0=0.3, BlendTime_1=0.3, bActiveValue=@n4) {blendType=Linear}
     BlendTime_0=0.3
     BlendTime_1=0.3
     bActiveValue=Get bIsSkyDiving()
    ) {blendType=Linear}) {layerGroup=None}
   AnimGraphNode_StateMachine_0 = WingPosition() {is=title AnimGraphNode_StateMachine}
   Save cached pose 'preSprint'(Pose=@AnimGraphNode_StateMachine_0) {cacheName=preSprint}
  ]}
 WingPosition {
  entry=Base
  path=AnimGraph.AnimGraphNode_StateMachine_0.WingPosition
  ids=id
  nodes [
   AnimStateNode_0 = Base() {is=title AnimStateNode}
   AnimStateNode_1 = Targeting() {is=title AnimStateNode}
   AnimStateNode_2 = Crouching() {is=title AnimStateNode}
   AnimStateNode_3 = Falling() {is=title AnimStateNode}
   AnimStateNode_4 = Jump() {is=title AnimStateNode}
   AnimStateNode_5 = SkydiveMode() {is=title AnimStateNode}
   AnimStateTransitionNode_0 = AnimStateTransitionNode()
   AnimStateTransitionNode_1 = AnimStateTransitionNode()
   AnimStateTransitionNode_10 = AnimStateTransitionNode()
   AnimStateTransitionNode_11 = AnimStateTransitionNode()
   AnimStateTransitionNode_2 = AnimStateTransitionNode()
   AnimStateTransitionNode_3 = AnimStateTransitionNode()
   AnimStateTransitionNode_4 = AnimStateTransitionNode()
   AnimStateTransitionNode_5 = AnimStateTransitionNode()
   AnimStateTransitionNode_6 = AnimStateTransitionNode()
   AnimStateTransitionNode_7 = AnimStateTransitionNode()
   AnimStateTransitionNode_8 = AnimStateTransitionNode()
   AnimStateTransitionNode_9 = AnimStateTransitionNode()
  ]
  links [
   {
    dst=Crouching
    rule=AnimGraph.AnimGraphNode_StateMachine_0.WingPosition.AnimStateTransitionNode_1.Transition
    src=Base
    via=AnimStateTransitionNode_1
   }
   {
    dst=Falling
    rule=AnimGraph.AnimGraphNode_StateMachine_0.WingPosition.AnimStateTransitionNode_3.Transition
    src=Base
    via=AnimStateTransitionNode_3
   }
   {
    dst=Jump
    rule=AnimGraph.AnimGraphNode_StateMachine_0.WingPosition.AnimStateTransitionNode_7.Transition
    src=Base
    via=AnimStateTransitionNode_7
   }
   {
    dst=SkydiveMode
    rule=AnimGraph.AnimGraphNode_StateMachine_0.WingPosition.AnimStateTransitionNode_9.Transition
    src=Base
    via=AnimStateTransitionNode_9
   }
   {
    dst=Targeting
    rule=AnimGraph.AnimGraphNode_StateMachine_0.WingPosition.AnimStateTransitionNode_0.Transition
    src=Base
    via=AnimStateTransitionNode_0
   }
   {
    dst=Base
    rule=AnimGraph.AnimGraphNode_StateMachine_0.WingPosition.AnimStateTransitionNode_6.Transition
    src=Crouching
    via=AnimStateTransitionNode_6
   }
   {
    dst=Base
    rule=AnimGraph.AnimGraphNode_StateMachine_0.WingPosition.AnimStateTransitionNode_4.Transition
    src=Falling
    via=AnimStateTransitionNode_4
   }
   {
    dst=Base
    rule=AnimGraph.AnimGraphNode_StateMachine_0.WingPosition.AnimStateTransitionNode_8.Transition
    src=Jump
    via=AnimStateTransitionNode_8
   }
   {
    dst=Falling
    rule=AnimGraph.AnimGraphNode_StateMachine_0.WingPosition.AnimStateTransitionNode_2.Transition
    src=Jump
    via=AnimStateTransitionNode_2
   }
   {
    dst=SkydiveMode
    rule=AnimGraph.AnimGraphNode_StateMachine_0.WingPosition.AnimStateTransitionNode_11.Transition
    src=Jump
    via=AnimStateTransitionNode_11
   }
   {
    dst=Base
    rule=AnimGraph.AnimGraphNode_StateMachine_0.WingPosition.AnimStateTransitionNode_10.Transition
    src=SkydiveMode
    via=AnimStateTransitionNode_10
   }
   {
    dst=Base
    rule=AnimGraph.AnimGraphNode_StateMachine_0.WingPosition.AnimStateTransitionNode_5.Transition
    src=Targeting
    via=AnimStateTransitionNode_5
   }
  ]
 }
 Base {
  path=AnimGraph.AnimGraphNode_StateMachine_0.WingPosition.AnimStateNode_0.Base
  nodes [
   n1 = Get PlayerFwdBwd() -> (PlayerFwdBwd:float)
   n2 = Get PlayerLeftRight() -> (PlayerLeftRight:float)
   Output Animation Pose(Result=Apply Additive(
     Base=Wing_Movement_Poses_PoseAsset : Base!(PoseWeight:float=1) -> (Pose:Pose Link Structure) {
      is=title AnimGraphNode_PoseByName
      poseAsset=/BRCosmetics/Accessories/FORT_Backpacks/F_MED_Aztec_Backpack/Animations/Wing_Movement_Poses_PoseAsset
      poseName=Base
     }
     Additive=Blendspace Player 'BS_Wing_Leans_Ground'(X=@n2, Y=@n1) {blendSpace=/BRCosmetics/Accessories/FORT_Backpacks/F_MED_Aztec_Backpack/Animations/BS_Wing_Leans_Ground}
    ) {alphaBoolBlend {blendOption=Linear}})
  ]
 }
 Targeting {
  path=AnimGraph.AnimGraphNode_StateMachine_0.WingPosition.AnimStateNode_1.Targeting
  nodes [Output Animation Pose(Result=Wing_Movement_Poses_PoseAsset : Targeting!(PoseWeight:float=1) -> (Pose:Pose Link Structure) {
     is=title AnimGraphNode_PoseByName
     poseAsset=&Wing_Movement_Poses_PoseAsset
     poseName=Targeting
    })]
 }
 Crouching {
  path=AnimGraph.AnimGraphNode_StateMachine_0.WingPosition.AnimStateNode_2.Crouching
  nodes [Output Animation Pose(Result=Wing_Movement_Poses_PoseAsset : Crouch!(PoseWeight:float=1) -> (Pose:Pose Link Structure) {
     is=title AnimGraphNode_PoseByName
     poseAsset=&Wing_Movement_Poses_PoseAsset
     poseName=Crouch
    })]
 }
 Falling {
  path=AnimGraph.AnimGraphNode_StateMachine_0.WingPosition.AnimStateNode_3.Falling
  nodes [Output Animation Pose(Result=Wing_Movement_Poses_PoseAsset : Jump!(PoseWeight:float=1) -> (Pose:Pose Link Structure) {
     is=title AnimGraphNode_PoseByName
     poseAsset=&Wing_Movement_Poses_PoseAsset
     poseName=Jump
    })]
 }
 Jump {
  path=AnimGraph.AnimGraphNode_StateMachine_0.WingPosition.AnimStateNode_4.Jump
  nodes [Output Animation Pose(Result=Wing_Movement_Poses_PoseAsset : Jump(PoseWeight=1) {poseAsset=&Wing_Movement_Poses_PoseAsset, poseName=Jump})]
 }
 SkydiveMode {
  path=AnimGraph.AnimGraphNode_StateMachine_0.WingPosition.AnimStateNode_5.SkydiveMode
  nodes [
   AnimGraphNode_StateMachine_0 = Skydiving State() {is=title AnimGraphNode_StateMachine}
   Output Animation Pose(Result=@AnimGraphNode_StateMachine_0)
  ]
 }
 Skydiving State {
  entry=Parachuting
  path=AnimGraph.AnimGraphNode_StateMachine_0.WingPosition.AnimStateNode_5.SkydiveMode.AnimGraphNode_StateMachine_0.Skydiving State
  ids=id
  nodes [
   AnimStateNode_0 = Parachuting() {is=title AnimStateNode}
   AnimStateNode_1 = Diving() {is=title AnimStateNode}
   AnimStateNode_2 = State_2() {is=title AnimStateNode}
   AnimStateTransitionNode_3 = AnimStateTransitionNode()
   AnimStateTransitionNode_4 = AnimStateTransitionNode()
   AnimStateTransitionNode_5 = AnimStateTransitionNode()
   AnimStateTransitionNode_6 = AnimStateTransitionNode()
   AnimStateTransitionNode_7 = AnimStateTransitionNode()
   AnimStateTransitionNode_8 = AnimStateTransitionNode()
  ]
  links [
   {
    dst=Parachuting
    rule=AnimGraph.AnimGraphNode_StateMachine_0.WingPosition.AnimStateNode_5.SkydiveMode.AnimGraphNode_StateMachine_0.Skydiving State.AnimStateTransitionNode_6.Transition
    src=Diving
    via=AnimStateTransitionNode_6
   }
   {
    dst=State_2
    rule=AnimGraph.AnimGraphNode_StateMachine_0.WingPosition.AnimStateNode_5.SkydiveMode.AnimGraphNode_StateMachine_0.Skydiving State.AnimStateTransitionNode_4.Transition
    src=Diving
    via=AnimStateTransitionNode_4
   }
   {
    dst=Diving
    rule=AnimGraph.AnimGraphNode_StateMachine_0.WingPosition.AnimStateNode_5.SkydiveMode.AnimGraphNode_StateMachine_0.Skydiving State.AnimStateTransitionNode_3.Transition
    src=Parachuting
    via=AnimStateTransitionNode_3
   }
   {
    dst=State_2
    rule=AnimGraph.AnimGraphNode_StateMachine_0.WingPosition.AnimStateNode_5.SkydiveMode.AnimGraphNode_StateMachine_0.Skydiving State.AnimStateTransitionNode_7.Transition
    src=Parachuting
    via=AnimStateTransitionNode_7
   }
   {
    dst=Diving
    rule=AnimGraph.AnimGraphNode_StateMachine_0.WingPosition.AnimStateNode_5.SkydiveMode.AnimGraphNode_StateMachine_0.Skydiving State.AnimStateTransitionNode_8.Transition
    src=State_2
    via=AnimStateTransitionNode_8
   }
   {
    dst=Parachuting
    rule=AnimGraph.AnimGraphNode_StateMachine_0.WingPosition.AnimStateNode_5.SkydiveMode.AnimGraphNode_StateMachine_0.Skydiving State.AnimStateTransitionNode_5.Transition
    src=State_2
    via=AnimStateTransitionNode_5
   }
  ]
 }
 Parachuting {
  path=AnimGraph.AnimGraphNode_StateMachine_0.WingPosition.AnimStateNode_5.SkydiveMode.AnimGraphNode_StateMachine_0.Skydiving State.AnimStateNode_0.Parachuting
  nodes [Output Animation Pose(Result=Wing_Movement_Poses_PoseAsset : Float!(PoseWeight:float=0.25) -> (Pose:Pose Link Structure) {
     is=title AnimGraphNode_PoseByName
     poseAsset=&Wing_Movement_Poses_PoseAsset
     poseName=Float
    })]
 }
 Diving {
  path=AnimGraph.AnimGraphNode_StateMachine_0.WingPosition.AnimStateNode_5.SkydiveMode.AnimGraphNode_StateMachine_0.Skydiving State.AnimStateNode_1.Diving
  nodes [Output Animation Pose(Result=Wing_Movement_Poses_PoseAsset : Dive!(PoseWeight:float=1) -> (Pose:Pose Link Structure) {
     is=title AnimGraphNode_PoseByName
     poseAsset=&Wing_Movement_Poses_PoseAsset
     poseName=Dive
    })]
 }
 State_2 {
  path=AnimGraph.AnimGraphNode_StateMachine_0.WingPosition.AnimStateNode_5.SkydiveMode.AnimGraphNode_StateMachine_0.Skydiving State.AnimStateNode_2.State_2
  nodes [Output Animation Pose(Result=Wing_Movement_Poses_PoseAsset : Float(PoseWeight=1) {poseAsset=&Wing_Movement_Poses_PoseAsset, poseName=Float})]
 }
 Transition {
  path=AnimGraph.AnimGraphNode_StateMachine_0.WingPosition.AnimStateNode_5.SkydiveMode.AnimGraphNode_StateMachine_0.Skydiving State.AnimStateTransitionNode_3.Transition
  nodes [Result(bCanEnterTransition=NOT Boolean(A=Get bIsParachuteOpened()))]
 }
 Transition {
  path=AnimGraph.AnimGraphNode_StateMachine_0.WingPosition.AnimStateNode_5.SkydiveMode.AnimGraphNode_StateMachine_0.Skydiving State.AnimStateTransitionNode_4.Transition
  nodes [Result(bCanEnterTransition=NOT Boolean(A=Get bIsSkydiveDiveMode()))]
 }
 Transition {
  path=AnimGraph.AnimGraphNode_StateMachine_0.WingPosition.AnimStateNode_5.SkydiveMode.AnimGraphNode_StateMachine_0.Skydiving State.AnimStateTransitionNode_5.Transition
  nodes [Result(bCanEnterTransition=Get bIsParachuteOpened())]
 }
 Transition {
  path=AnimGraph.AnimGraphNode_StateMachine_0.WingPosition.AnimStateNode_5.SkydiveMode.AnimGraphNode_StateMachine_0.Skydiving State.AnimStateTransitionNode_6.Transition
  nodes [Result(bCanEnterTransition=Get bIsParachuteOpened())]
 }
 Transition {
  path=AnimGraph.AnimGraphNode_StateMachine_0.WingPosition.AnimStateNode_5.SkydiveMode.AnimGraphNode_StateMachine_0.Skydiving State.AnimStateTransitionNode_7.Transition
  nodes [Result(bCanEnterTransition=NOT Boolean(A=Get bIsParachuteOpened()))]
 }
 Transition {
  path=AnimGraph.AnimGraphNode_StateMachine_0.WingPosition.AnimStateNode_5.SkydiveMode.AnimGraphNode_StateMachine_0.Skydiving State.AnimStateTransitionNode_8.Transition
  nodes [Result(bCanEnterTransition=Get bIsSkydiveDiveMode())]
 }
 Transition {
  path=AnimGraph.AnimGraphNode_StateMachine_0.WingPosition.AnimStateTransitionNode_0.Transition
  nodes [Result(bCanEnterTransition=Get bIsTargeting())]
 }
 Transition {
  path=AnimGraph.AnimGraphNode_StateMachine_0.WingPosition.AnimStateTransitionNode_1.Transition
  nodes [Result(bCanEnterTransition=Get bIsCrouching())]
 }
 Transition {
  path=AnimGraph.AnimGraphNode_StateMachine_0.WingPosition.AnimStateTransitionNode_10.Transition
  nodes [Result(bCanEnterTransition=NOT Boolean(A=Get bIsSkyDiving()))]
 }
 Transition {
  path=AnimGraph.AnimGraphNode_StateMachine_0.WingPosition.AnimStateTransitionNode_11.Transition
  nodes [Result(bCanEnterTransition=Get bIsSkyDiving())]
 }
 Transition {
  path=AnimGraph.AnimGraphNode_StateMachine_0.WingPosition.AnimStateTransitionNode_2.Transition
  nodes [Result(bCanEnterTransition=Get bIsFalling())]
 }
 Transition {
  path=AnimGraph.AnimGraphNode_StateMachine_0.WingPosition.AnimStateTransitionNode_3.Transition
  nodes [Result(bCanEnterTransition=Get bIsFalling())]
 }
 Transition {
  path=AnimGraph.AnimGraphNode_StateMachine_0.WingPosition.AnimStateTransitionNode_4.Transition
  nodes [Result(bCanEnterTransition=NOT Boolean(A=Get bIsFalling()))]
 }
 Transition {
  path=AnimGraph.AnimGraphNode_StateMachine_0.WingPosition.AnimStateTransitionNode_5.Transition
  nodes [Result(bCanEnterTransition=NOT Boolean(A=Get bIsTargeting()))]
 }
 Transition {
  path=AnimGraph.AnimGraphNode_StateMachine_0.WingPosition.AnimStateTransitionNode_6.Transition
  nodes [Result(bCanEnterTransition=NOT Boolean(A=Get bIsCrouching()))]
 }
 Transition {
  path=AnimGraph.AnimGraphNode_StateMachine_0.WingPosition.AnimStateTransitionNode_7.Transition
  nodes [Result(bCanEnterTransition=Get bIsJumping())]
 }
 Transition {
  path=AnimGraph.AnimGraphNode_StateMachine_0.WingPosition.AnimStateTransitionNode_8.Transition
  nodes [Result(bCanEnterTransition=NOT Boolean(A=Get bIsJumping()))]
 }
 Transition {
  path=AnimGraph.AnimGraphNode_StateMachine_0.WingPosition.AnimStateTransitionNode_9.Transition
  nodes [Result(bCanEnterTransition=Get bIsSkyDiving())]
 }
 EventGraph {links [], nodes []}
]
variables []
